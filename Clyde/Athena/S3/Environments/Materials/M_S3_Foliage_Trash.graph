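# Default Lit material for scattered trash foliage in the S3 environments, usable on instanced static meshes and Nanite, with diffuse, mask and normal textures.
# An optional switch tints masked areas by multiplying diffuse red with a partly desaturated gradient, whose lookup combines mask green and per-instance random for color variation.
# A camera-centered sphere mask builds a Z offset that drops to negative beyond the fade radius, apparently sinking distant pieces.
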Material /Clyde/Athena/S3/Environments/Materials/M_S3_Foliage_Trash
graph {
 properties {bUsedWithInstancedStaticMeshes=true, bUsedWithNanite=true, pixelDepthOffsetMode=PDOM_Legacy, usedShadingModels=Default Lit}
 nodes [
  n1 = WorldPosition()
  n2 = CameraPositionWS()
  n3 = ScalarParameter() {defaultValue=5000, parameterName=FadeRadius}
  n4 = ScalarParameter() {defaultValue=20, parameterName=FadeHardness}
  AppendVector(
   A=Constant2Vector()
   B=LinearInterpolate(A=Constant() {r=-500}, B=Constant(), Alpha=SphereMask(A=@n1.XYZ, B=@n2, Radius=@n3, Hardness=@n4) {hardnessPercent=20})
  )
  n5 = TextureSampleParameter2D() {parameterName=Diffuse, samplerSource=SSM_Wrap_WorldGroupSettings, texture=/Game/Environments/Props/Trash/Textures/T_FORT_Scraps_UE4_D}
  n6 = PerInstanceRandom()
  n7 = TextureSampleParameter2D() {
   parameterName=ColorMask
   samplerSource=SSM_Wrap_WorldGroupSettings
   samplerType=SAMPLERTYPE_LinearColor
   texture=/Clyde/Athena/S3/Environments/Landscape/Textures/Foliage/T_FORT_Scraps_Mask_01
  }
  n8 = TextureSampleParameter2D(UVs=Add(A=Multiply(A=@n6, B=@n7.G), B=@n7.G)) {parameterName=Gradient, samplerSource=SSM_Wrap_WorldGroupSettings, texture=/Game/Environments/Props/Retail/Textures/ChairColors_01}
  n9 = Desaturation(None=@n8.RGB, Fraction=ScalarParameter() {defaultValue=0.5, parameterName=DesaturationAmount})
  StaticSwitchParameter(True=LinearInterpolate(A=@n5.RGB, B=Multiply(A=@n5.R, B=@n9), Alpha=@n7.R), False=@n5.RGB) {defaultValue=true, parameterName=UseGradientColor}
  TextureSampleParameter2D() {
   parameterName=Masks
   samplerSource=SSM_Wrap_WorldGroupSettings
   samplerType=SAMPLERTYPE_Masks
   texture=/Game/Environments/Props/Trash/Textures/T_FORT_Scraps_UE4_S
  }
  TextureSampleParameter2D() {
   parameterName=Normal
   samplerSource=SSM_Wrap_WorldGroupSettings
   samplerType=SAMPLERTYPE_Normal
   texture=/Game/Environments/Props/Trash/Textures/T_FORT_Scraps_UE4_N
  }
 ]
}
